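# Sound cue for the WhisperRemedySweet Sprite's bouncing sound, layering a vocal line with a water effect.
# A Mixer combines two Modulator branches, each fed by an equally weighted Random node.
# The voice branch picks from patrol bouncing and reused inventory-drop lines and is excluded from branch culling; the other picks water grab effects.
# It plays as a creature vocalization with close sprite VO attenuation and a shared sprite concurrency set.
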SoundCue /Sprites_WhisperRemedySweetSprite/Assets/Sounds/Cues/VO/SC_VO_S33_WhisperRemedySweetSprite_Bouncing
root=Mixer_0
nodes [
 {id=Mixer_0, children [Modulator_0,Modulator_1], properties {inputVolume [1,1]}}
 {id=Modulator_0, children [Random_1]}
 {id=Modulator_1, children [Random_0]}
 {id=Random_0, children [WavePlayer_2, WavePlayer_3, WavePlayer_4, WavePlayer_11, WavePlayer_12], properties {weights [1,1,1,1,1]}}
 {
  id=Random_1
  children [WavePlayer_5,WavePlayer_6,WavePlayer_7,WavePlayer_0,WavePlayer_1]
  properties {bShouldExcludeFromBranchCulling=true, weights [1,1,1,1,1]}
 }
 {
  id=WavePlayer_0
  properties {soundWaveAssetPtr=/Sprites_WhisperRemedySweetSprite/Assets/Sounds/Waves/VO/VO_S33_WhisperRemedySweetSprite_Inventory_Dropped_010}
 }
 {
  id=WavePlayer_1
  properties {soundWaveAssetPtr=/Sprites_WhisperRemedySweetSprite/Assets/Sounds/Waves/VO/VO_S33_WhisperRemedySweetSprite_Inventory_Dropped_013}
 }
 {id=WavePlayer_11, properties {soundWaveAssetPtr=/Sprites_WhisperRemedySweetSprite/Assets/Sounds/Waves/SFX/Sprite_Water_Grab_02}}
 {id=WavePlayer_12, properties {soundWaveAssetPtr=/Sprites_WhisperRemedySweetSprite/Assets/Sounds/Waves/SFX/Sprite_Water_Grab_05}}
 {id=WavePlayer_2, properties {soundWaveAssetPtr=/Sprites_WhisperRemedySweetSprite/Assets/Sounds/Waves/SFX/Sprite_Water_Grab_08}}
 {id=WavePlayer_3, properties {soundWaveAssetPtr=/Sprites_WhisperRemedySweetSprite/Assets/Sounds/Waves/SFX/Sprite_Water_Grab_07}}
 {id=WavePlayer_4, properties {soundWaveAssetPtr=/Sprites_WhisperRemedySweetSprite/Assets/Sounds/Waves/SFX/Sprite_Water_Grab_01}}
 {
  id=WavePlayer_5
  properties {soundWaveAssetPtr=/Sprites_WhisperRemedySweetSprite/Assets/Sounds/Waves/VO/VO_S33_WhisperRemedySweetSprite_Patrol_Bouncing_010}
 }
 {
  id=WavePlayer_6
  properties {soundWaveAssetPtr=/Sprites_WhisperRemedySweetSprite/Assets/Sounds/Waves/VO/VO_S33_WhisperRemedySweetSprite_Patrol_Bouncing_011}
 }
 {
  id=WavePlayer_7
  properties {soundWaveAssetPtr=/Sprites_WhisperRemedySweetSprite/Assets/Sounds/Waves/VO/VO_S33_WhisperRemedySweetSprite_Patrol_Bouncing_012}
 }
]
properties {
 attenuationSettings=/Sprites_WhisperRemedySweetSprite/Assets/Sounds/Settings/AS_Sprite_VO_Close
 maxDistance=4000
 priority=2
 soundClassObject=/Game/Sounds/FortSoundClassesAndModes/Fort_Creatures_Vocalization_Normal
 volumeMultiplier=1
 concurrencySet [/Sprites/Assets/Sounds/Music/CS_Sprite_Pri_MusicVox]
}
