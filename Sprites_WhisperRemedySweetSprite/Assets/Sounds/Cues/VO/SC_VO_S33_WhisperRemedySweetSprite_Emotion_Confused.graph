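# Sound cue for the Whisper Remedy Sweet Sprite's confused emotion vocalization, classed as creature vocalization with close VO attenuation and shared Sprite concurrency.
# A Modulator varies playback and feeds from a Quality Level switch over two Random pools of voice waves.
# The first two quality slots pick from eight takes, while the last two use a four-take subset.
# The larger pool is excluded from branch culling.
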SoundCue /Sprites_WhisperRemedySweetSprite/Assets/Sounds/Cues/VO/SC_VO_S33_WhisperRemedySweetSprite_Emotion_Confused
root=Modulator_0
nodes [
 {id=Modulator_0, children [QualityLevel_0], properties {volumeMin=0.6}}
 {id=QualityLevel_0, children [Random_2,Random_2,Random_4,Random_4]}
 {
  id=Random_2
  children [WavePlayer_13, WavePlayer_14, WavePlayer_15, WavePlayer_16, WavePlayer_17, WavePlayer_18, WavePlayer_19, WavePlayer_20]
  properties {bShouldExcludeFromBranchCulling=true, weights [1,1,1,1,1,1,1,1]}
 }
 {id=Random_4, children [WavePlayer_23, WavePlayer_24, WavePlayer_22, WavePlayer_21], properties {weights [1,1,1,1]}}
 {
  id=WavePlayer_13
  properties {soundWaveAssetPtr=/Sprites_WhisperRemedySweetSprite/Assets/Sounds/Waves/VO/VO_S33_WhisperRemedySweetSprite_Emotion_Confused_010}
 }
 {
  id=WavePlayer_14
  properties {soundWaveAssetPtr=/Sprites_WhisperRemedySweetSprite/Assets/Sounds/Waves/VO/VO_S33_WhisperRemedySweetSprite_Emotion_Confused_011}
 }
 {
  id=WavePlayer_15
  properties {soundWaveAssetPtr=/Sprites_WhisperRemedySweetSprite/Assets/Sounds/Waves/VO/VO_S33_WhisperRemedySweetSprite_Emotion_Confused_012}
 }
 {
  id=WavePlayer_16
  properties {soundWaveAssetPtr=/Sprites_WhisperRemedySweetSprite/Assets/Sounds/Waves/VO/VO_S33_WhisperRemedySweetSprite_Emotion_Confused_013}
 }
 {
  id=WavePlayer_17
  properties {soundWaveAssetPtr=/Sprites_WhisperRemedySweetSprite/Assets/Sounds/Waves/VO/VO_S33_WhisperRemedySweetSprite_Emotion_Confused_014}
 }
 {
  id=WavePlayer_18
  properties {soundWaveAssetPtr=/Sprites_WhisperRemedySweetSprite/Assets/Sounds/Waves/VO/VO_S33_WhisperRemedySweetSprite_Emotion_Confused_015}
 }
 {
  id=WavePlayer_19
  properties {soundWaveAssetPtr=/Sprites_WhisperRemedySweetSprite/Assets/Sounds/Waves/VO/VO_S33_WhisperRemedySweetSprite_Emotion_Confused_016}
 }
 {
  id=WavePlayer_20
  properties {soundWaveAssetPtr=/Sprites_WhisperRemedySweetSprite/Assets/Sounds/Waves/VO/VO_S33_WhisperRemedySweetSprite_Emotion_Confused_017}
 }
 {
  id=WavePlayer_21
  properties {soundWaveAssetPtr=&VO_S33_WhisperRemedySweetSprite_Emotion_Confused_017}
 }
 {
  id=WavePlayer_22
  properties {soundWaveAssetPtr=&VO_S33_WhisperRemedySweetSprite_Emotion_Confused_015}
 }
 {
  id=WavePlayer_23
  properties {soundWaveAssetPtr=&VO_S33_WhisperRemedySweetSprite_Emotion_Confused_010}
 }
 {
  id=WavePlayer_24
  properties {soundWaveAssetPtr=&VO_S33_WhisperRemedySweetSprite_Emotion_Confused_011}
 }
]
properties {
 attenuationSettings=/Sprites_WhisperRemedySweetSprite/Assets/Sounds/Settings/AS_Sprite_VO_Close
 maxDistance=4000
 soundClassObject=/Game/Sounds/FortSoundClassesAndModes/Fort_Creatures_Vocalization_Normal
 volumeMultiplier=1
 concurrencySet [/Sprites/Assets/Sounds/Music/CS_Sprite_Pri_MusicVox]
}
