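# Sound Cue for the Whisper Remedy Sweet Sprite's "dropped" emotion voice line, built on inventory-dropped VO recordings.
# A Random node picks among five equally weighted waves and is excluded from branch culling, feeding a Modulator with a lowered minimum volume for playback variation.
# It uses close-range sprite VO attenuation, the creature vocalization sound class and a shared sprite music-vox priority concurrency set.
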SoundCue /Sprites_WhisperRemedySweetSprite/Assets/Sounds/Cues/VO/SC_VO_S33_WhisperRemedySweetSprite_Emotion_Dropped
root=Modulator_0
nodes [
 {id=Modulator_0, children [Random_1], properties {volumeMin=0.6}}
 {
  id=Random_1
  children [WavePlayer_5,WavePlayer_6,WavePlayer_7,WavePlayer_8,WavePlayer_9]
  properties {bShouldExcludeFromBranchCulling=true, weights [1,1,1,1,1]}
 }
 {
  id=WavePlayer_5
  properties {soundWaveAssetPtr=/Sprites_WhisperRemedySweetSprite/Assets/Sounds/Waves/VO/VO_S33_WhisperRemedySweetSprite_Inventory_Dropped_010}
 }
 {
  id=WavePlayer_6
  properties {soundWaveAssetPtr=/Sprites_WhisperRemedySweetSprite/Assets/Sounds/Waves/VO/VO_S33_WhisperRemedySweetSprite_Inventory_Dropped_011}
 }
 {
  id=WavePlayer_7
  properties {soundWaveAssetPtr=/Sprites_WhisperRemedySweetSprite/Assets/Sounds/Waves/VO/VO_S33_WhisperRemedySweetSprite_Inventory_Dropped_012}
 }
 {
  id=WavePlayer_8
  properties {soundWaveAssetPtr=/Sprites_WhisperRemedySweetSprite/Assets/Sounds/Waves/VO/VO_S33_WhisperRemedySweetSprite_Inventory_Dropped_013}
 }
 {
  id=WavePlayer_9
  properties {soundWaveAssetPtr=/Sprites_WhisperRemedySweetSprite/Assets/Sounds/Waves/VO/VO_S33_WhisperRemedySweetSprite_Inventory_Dropped_014}
 }
]
properties {
 attenuationSettings=/Sprites_WhisperRemedySweetSprite/Assets/Sounds/Settings/AS_Sprite_VO_Close
 maxDistance=4000
 soundClassObject=/Game/Sounds/FortSoundClassesAndModes/Fort_Creatures_Vocalization_Normal
 volumeMultiplier=1
 concurrencySet [/Sprites/Assets/Sounds/Music/CS_Sprite_Pri_MusicVox]
}
