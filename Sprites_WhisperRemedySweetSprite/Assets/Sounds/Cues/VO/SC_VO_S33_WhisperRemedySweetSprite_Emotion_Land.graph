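# Sound Cue for the Whisper Remedy Sweet Sprite's Land emotion vocalization, drawing on its Inventory_Land voice clips.
# A Quality Level node sends the first two levels to a ten-clip random pool, excluded from branch culling, and the last two to a four-clip subset.
# A root Modulator randomizes volume per play.
# Close-range Sprite attenuation, the creature vocalization class and shared music-vox concurrency govern playback.
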SoundCue /Sprites_WhisperRemedySweetSprite/Assets/Sounds/Cues/VO/SC_VO_S33_WhisperRemedySweetSprite_Emotion_Land
root=Modulator_0
nodes [
 {id=Modulator_0, children [QualityLevel_0], properties {volumeMin=0.6}}
 {id=QualityLevel_0, children [Random_0,Random_0,Random_2,Random_2]}
 {
  id=Random_0
  children [
   WavePlayer_0
   WavePlayer_1
   WavePlayer_2
   WavePlayer_3
   WavePlayer_4
   WavePlayer_10
   WavePlayer_11
   WavePlayer_12
   WavePlayer_13
   WavePlayer_14
  ]
  properties {bShouldExcludeFromBranchCulling=true, weights [1,1,1,1,1,1,1,1,1,1]}
 }
 {id=Random_2, children [WavePlayer_18, WavePlayer_17, WavePlayer_16, WavePlayer_15], properties {weights [1,1,1,1]}}
 {
  id=WavePlayer_0
  properties {soundWaveAssetPtr=/Sprites_WhisperRemedySweetSprite/Assets/Sounds/Waves/VO/VO_S33_WhisperRemedySweetSprite_Inventory_Land_010}
 }
 {
  id=WavePlayer_1
  properties {soundWaveAssetPtr=/Sprites_WhisperRemedySweetSprite/Assets/Sounds/Waves/VO/VO_S33_WhisperRemedySweetSprite_Inventory_Land_011}
 }
 {
  id=WavePlayer_10
  properties {soundWaveAssetPtr=/Sprites_WhisperRemedySweetSprite/Assets/Sounds/Waves/VO/VO_S33_WhisperRemedySweetSprite_Inventory_Land_015}
 }
 {
  id=WavePlayer_11
  properties {soundWaveAssetPtr=/Sprites_WhisperRemedySweetSprite/Assets/Sounds/Waves/VO/VO_S33_WhisperRemedySweetSprite_Inventory_Land_016}
 }
 {
  id=WavePlayer_12
  properties {soundWaveAssetPtr=/Sprites_WhisperRemedySweetSprite/Assets/Sounds/Waves/VO/VO_S33_WhisperRemedySweetSprite_Inventory_Land_017}
 }
 {
  id=WavePlayer_13
  properties {soundWaveAssetPtr=/Sprites_WhisperRemedySweetSprite/Assets/Sounds/Waves/VO/VO_S33_WhisperRemedySweetSprite_Inventory_Land_018}
 }
 {
  id=WavePlayer_14
  properties {soundWaveAssetPtr=/Sprites_WhisperRemedySweetSprite/Assets/Sounds/Waves/VO/VO_S33_WhisperRemedySweetSprite_Inventory_Land_019}
 }
 {
  id=WavePlayer_15
  properties {soundWaveAssetPtr=/Sprites_WhisperRemedySweetSprite/Assets/Sounds/Waves/VO/VO_S33_WhisperRemedySweetSprite_Inventory_Land_014}
 }
 {
  id=WavePlayer_16
  properties {soundWaveAssetPtr=/Sprites_WhisperRemedySweetSprite/Assets/Sounds/Waves/VO/VO_S33_WhisperRemedySweetSprite_Inventory_Land_013}
 }
 {
  id=WavePlayer_17
  properties {soundWaveAssetPtr=&VO_S33_WhisperRemedySweetSprite_Inventory_Land_011}
 }
 {
  id=WavePlayer_18
  properties {soundWaveAssetPtr=&VO_S33_WhisperRemedySweetSprite_Inventory_Land_010}
 }
 {
  id=WavePlayer_2
  properties {soundWaveAssetPtr=/Sprites_WhisperRemedySweetSprite/Assets/Sounds/Waves/VO/VO_S33_WhisperRemedySweetSprite_Inventory_Land_012}
 }
 {
  id=WavePlayer_3
  properties {soundWaveAssetPtr=&VO_S33_WhisperRemedySweetSprite_Inventory_Land_013}
 }
 {
  id=WavePlayer_4
  properties {soundWaveAssetPtr=&VO_S33_WhisperRemedySweetSprite_Inventory_Land_014}
 }
]
properties {
 attenuationSettings=/Sprites_WhisperRemedySweetSprite/Assets/Sounds/Settings/AS_Sprite_VO_Close
 maxDistance=4000
 soundClassObject=/Game/Sounds/FortSoundClassesAndModes/Fort_Creatures_Vocalization_Normal
 volumeMultiplier=0.7
 concurrencySet [/Sprites/Assets/Sounds/Music/CS_Sprite_Pri_MusicVox]
}
