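# Sound cue for the Whisper Remedy Sweet Sprite's shrine goodbye voice line.
# A Random node picks one of six equally weighted VO waves, excluded from branch culling, and feeds a Modulator at the root.
# It uses the sprite's far VO attenuation, a tactical creature vocalization sound class, a shared sprite music/vox concurrency set, and a boosted volume multiplier.
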SoundCue /Sprites_WhisperRemedySweetSprite/Assets/Sounds/Cues/VO/SC_VO_S33_WhisperRemedySweetSprite_Shrine_Goodbye
root=Modulator_0
nodes [
 {id=Modulator_0, children [Random_0], properties {volumeMin=1}}
 {
  id=Random_0
  children [WavePlayer_0, WavePlayer_1, WavePlayer_2, WavePlayer_3, WavePlayer_4, WavePlayer_10]
  properties {bShouldExcludeFromBranchCulling=true, weights [1,1,1,1,1,1]}
 }
 {
  id=WavePlayer_0
  properties {soundWaveAssetPtr=/Sprites_WhisperRemedySweetSprite/Assets/Sounds/Waves/VO/VO_S33_WhisperRemedySweetSprite_Shrine_Goodbye_010}
 }
 {
  id=WavePlayer_1
  properties {soundWaveAssetPtr=/Sprites_WhisperRemedySweetSprite/Assets/Sounds/Waves/VO/VO_S33_WhisperRemedySweetSprite_Shrine_Goodbye_011}
 }
 {
  id=WavePlayer_10
  properties {soundWaveAssetPtr=/Sprites_WhisperRemedySweetSprite/Assets/Sounds/Waves/VO/VO_S33_WhisperRemedySweetSprite_Shrine_Goodbye_015}
 }
 {
  id=WavePlayer_2
  properties {soundWaveAssetPtr=/Sprites_WhisperRemedySweetSprite/Assets/Sounds/Waves/VO/VO_S33_WhisperRemedySweetSprite_Shrine_Goodbye_012}
 }
 {
  id=WavePlayer_3
  properties {soundWaveAssetPtr=/Sprites_WhisperRemedySweetSprite/Assets/Sounds/Waves/VO/VO_S33_WhisperRemedySweetSprite_Shrine_Goodbye_013}
 }
 {
  id=WavePlayer_4
  properties {soundWaveAssetPtr=/Sprites_WhisperRemedySweetSprite/Assets/Sounds/Waves/VO/VO_S33_WhisperRemedySweetSprite_Shrine_Goodbye_014}
 }
]
properties {
 attenuationSettings=/Sprites_WhisperRemedySweetSprite/Assets/Sounds/Settings/AS_Sprite_VO_Far
 maxDistance=5400
 soundClassObject=&Fort_Creatures_Vocalization_Tactical
 volumeMultiplier=1.5
 concurrencySet [/Sprites/Assets/Sounds/Music/CS_Sprite_Pri_MusicVox]
}
