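# Sound cue for the Whisper Remedy Sweet Sprite's shrine leap voice line, using "Approached_Leap" VO waves.
# A quality-level node routes the first two levels to a random pick of eight equally weighted waves and the last two to a four-wave subset.
# A root modulator varies volume per play.
# Playback uses the far Sprite VO attenuation, the creature vocalization sound class and a shared Sprite music-vox concurrency set.
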SoundCue /Sprites_WhisperRemedySweetSprite/Assets/Sounds/Cues/VO/SC_VO_S33_WhisperRemedySweetSprite_Shrine_Leap
root=Modulator_0
nodes [
 {id=Modulator_0, children [QualityLevel_0], properties {volumeMin=0.6}}
 {id=QualityLevel_0, children [Random_1,Random_1,Random_2,Random_2]}
 {
  id=Random_1
  children [WavePlayer_8, WavePlayer_9, WavePlayer_11, WavePlayer_13, WavePlayer_14, WavePlayer_15, WavePlayer_16, WavePlayer_17]
  properties {bShouldExcludeFromBranchCulling=true, weights [1,1,1,1,1,1,1,1]}
 }
 {
  id=Random_2
  children [WavePlayer_18, WavePlayer_19, WavePlayer_20, WavePlayer_21]
  properties {bShouldExcludeFromBranchCulling=true, weights [1,1,1,1]}
 }
 {
  id=WavePlayer_11
  properties {soundWaveAssetPtr=/Sprites_WhisperRemedySweetSprite/Assets/Sounds/Waves/VO/VO_S33_WhisperRemedySweetSprite_Approached_Leap_024}
 }
 {
  id=WavePlayer_13
  properties {soundWaveAssetPtr=/Sprites_WhisperRemedySweetSprite/Assets/Sounds/Waves/VO/VO_S33_WhisperRemedySweetSprite_Approached_Leap_033}
 }
 {
  id=WavePlayer_14
  properties {soundWaveAssetPtr=/Sprites_WhisperRemedySweetSprite/Assets/Sounds/Waves/VO/VO_S33_WhisperRemedySweetSprite_Approached_Leap_034}
 }
 {
  id=WavePlayer_15
  properties {soundWaveAssetPtr=/Sprites_WhisperRemedySweetSprite/Assets/Sounds/Waves/VO/VO_S33_WhisperRemedySweetSprite_Approached_Leap_035}
 }
 {
  id=WavePlayer_16
  properties {soundWaveAssetPtr=/Sprites_WhisperRemedySweetSprite/Assets/Sounds/Waves/VO/VO_S33_WhisperRemedySweetSprite_Approached_Leap_036}
 }
 {
  id=WavePlayer_17
  properties {soundWaveAssetPtr=/Sprites_WhisperRemedySweetSprite/Assets/Sounds/Waves/VO/VO_S33_WhisperRemedySweetSprite_Approached_Leap_037}
 }
 {
  id=WavePlayer_18
  properties {soundWaveAssetPtr=/Sprites_WhisperRemedySweetSprite/Assets/Sounds/Waves/VO/VO_S33_WhisperRemedySweetSprite_Approached_Leap_020}
 }
 {
  id=WavePlayer_19
  properties {soundWaveAssetPtr=&VO_S33_WhisperRemedySweetSprite_Approached_Leap_024}
 }
 {
  id=WavePlayer_20
  properties {soundWaveAssetPtr=&VO_S33_WhisperRemedySweetSprite_Approached_Leap_035}
 }
 {
  id=WavePlayer_21
  properties {soundWaveAssetPtr=&VO_S33_WhisperRemedySweetSprite_Approached_Leap_036}
 }
 {
  id=WavePlayer_8
  properties {soundWaveAssetPtr=&VO_S33_WhisperRemedySweetSprite_Approached_Leap_020}
 }
 {
  id=WavePlayer_9
  properties {soundWaveAssetPtr=/Sprites_WhisperRemedySweetSprite/Assets/Sounds/Waves/VO/VO_S33_WhisperRemedySweetSprite_Approached_Leap_021}
 }
]
properties {
 attenuationSettings=/Sprites_WhisperRemedySweetSprite/Assets/Sounds/Settings/AS_Sprite_VO_Far
 maxDistance=5400
 soundClassObject=/Game/Sounds/FortSoundClassesAndModes/Fort_Creatures_Vocalization_Normal
 volumeMultiplier=1
 concurrencySet [/Sprites/Assets/Sounds/Music/CS_Sprite_Pri_MusicVox]
}
