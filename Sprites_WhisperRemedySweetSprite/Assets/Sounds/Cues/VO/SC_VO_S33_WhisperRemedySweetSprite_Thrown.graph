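# Sound cue for throwing the Whisper Remedy Sweet Sprite, mixing voice, in-air loop and water unequip layers.
# The voice layer picks one of five thrown lines, passed through a modulator and a short fixed delay.
# The other layers randomly pick a looping in-air sound, which keeps the cue playing, and a one-shot water unequip sound.
# It uses projectile in-air loop attenuation and the creature vocalization sound class.
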SoundCue /Sprites_WhisperRemedySweetSprite/Assets/Sounds/Cues/VO/SC_VO_S33_WhisperRemedySweetSprite_Thrown
root=Mixer_0
nodes [
 {id=Delay_0, children [Modulator_0], properties {delayMax=0.1, delayMin=0.1}}
 {id=Mixer_0, children [Delay_0,Random_0,Random_2], properties {inputVolume [0.704,1.1453,1]}}
 {id=Modulator_0, children [Random_1], properties {volumeMin=0.8}}
 {id=Random_0, children [WavePlayer_0,WavePlayer_1,WavePlayer_2], properties {bShouldExcludeFromBranchCulling=true, weights [1,1,1]}}
 {
  id=Random_1
  children [WavePlayer_4,WavePlayer_5,WavePlayer_6,WavePlayer_7,WavePlayer_9]
  properties {bShouldExcludeFromBranchCulling=true, weights [1,1,1,1,1]}
 }
 {
  id=Random_2
  children [WavePlayer_3, WavePlayer_8, WavePlayer_10, WavePlayer_11]
  properties {bShouldExcludeFromBranchCulling=true, weights [1,1,1,1]}
 }
 {
  id=WavePlayer_0
  properties {bLooping=true, soundWaveAssetPtr=/Sprites_WhisperRemedySweetSprite/Assets/Sounds/Waves/SFX/WhisperRemedySweetSprite_InAir_Loop_01}
 }
 {
  id=WavePlayer_1
  properties {bLooping=true, soundWaveAssetPtr=/Sprites_WhisperRemedySweetSprite/Assets/Sounds/Waves/SFX/WhisperRemedySweetSprite_InAir_Loop_02}
 }
 {id=WavePlayer_10, properties {soundWaveAssetPtr=/Sprites_WhisperRemedySweetSprite/Assets/Sounds/Waves/SFX/Sprite_Water_Unequip_03}}
 {id=WavePlayer_11, properties {soundWaveAssetPtr=/Sprites_WhisperRemedySweetSprite/Assets/Sounds/Waves/SFX/Sprite_Water_Unequip_04}}
 {
  id=WavePlayer_2
  properties {bLooping=true, soundWaveAssetPtr=/Sprites_WhisperRemedySweetSprite/Assets/Sounds/Waves/SFX/WhisperRemedySweetSprite_InAir_Loop_03}
 }
 {id=WavePlayer_3, properties {soundWaveAssetPtr=/Sprites_WhisperRemedySweetSprite/Assets/Sounds/Waves/SFX/Sprite_Water_Unequip_01}}
 {
  id=WavePlayer_4
  properties {soundWaveAssetPtr=/Sprites_WhisperRemedySweetSprite/Assets/Sounds/Waves/VO/VO_S33_WhisperRemedySweetSprite_Inventory_Thrown_010}
 }
 {
  id=WavePlayer_5
  properties {soundWaveAssetPtr=/Sprites_WhisperRemedySweetSprite/Assets/Sounds/Waves/VO/VO_S33_WhisperRemedySweetSprite_Inventory_Thrown_011}
 }
 {
  id=WavePlayer_6
  properties {soundWaveAssetPtr=/Sprites_WhisperRemedySweetSprite/Assets/Sounds/Waves/VO/VO_S33_WhisperRemedySweetSprite_Inventory_Thrown_012}
 }
 {
  id=WavePlayer_7
  properties {soundWaveAssetPtr=/Sprites_WhisperRemedySweetSprite/Assets/Sounds/Waves/VO/VO_S33_WhisperRemedySweetSprite_Inventory_Thrown_013}
 }
 {id=WavePlayer_8, properties {soundWaveAssetPtr=/Sprites_WhisperRemedySweetSprite/Assets/Sounds/Waves/SFX/Sprite_Water_Unequip_02}}
 {
  id=WavePlayer_9
  properties {soundWaveAssetPtr=/Sprites_WhisperRemedySweetSprite/Assets/Sounds/Waves/VO/VO_S33_WhisperRemedySweetSprite_Inventory_Thrown_015}
 }
]
properties {
 attenuationSettings=/Game/Sounds/Attenuation/Fort_Weapon_Projectile_InAir_Loop_Att
 maxDistance=4500
 soundClassObject=/Game/Sounds/FortSoundClassesAndModes/Fort_Creatures_Vocalization_Normal
 volumeMultiplier=1
}
